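# CineMat is a simple Default Lit editor material in the camera folder, apparently for cinematic camera meshes.
# It exposes Base Color, Roughness and Specular parameters, defaulting to a dark grey, fairly glossy surface that material instances can adjust.
# It has no textures or math nodes and is flagged for use on spline meshes.
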Material /Engine/EditorMaterials/Camera/CineMat
graph {
 properties {bUsedWithSplineMeshes=true, pixelDepthOffsetMode=PDOM_Legacy, usedShadingModels=Default Lit}
 nodes [
  ScalarParameter() {defaultValue=0.2, parameterName=Roughness, sliderMax=1}
  ScalarParameter() {defaultValue=0.3, parameterName=Specular, sliderMax=1}
  VectorParameter() {parameterName=Base Color, defaultValue=<rgba 0.0431 0.0431 0.0431 0>}
 ]
}
